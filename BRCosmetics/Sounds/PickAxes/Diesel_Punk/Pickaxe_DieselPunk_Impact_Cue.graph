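# Sound cue for the impact of the Diesel Punk cosmetic pickaxe.
# A mixer layers a random Rocket Hammer sledge impact at reduced volume under a random Diesel Punk impact at full volume, with slight pitch variation on the latter only.
# A short fixed delay precedes the mix, and the whole cue is pitched slightly down.
# It primes on load and references shared pickaxe impact attenuation, property sheet and melee weapon concurrency assets.
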SoundCue /BRCosmetics/Sounds/PickAxes/Diesel_Punk/Pickaxe_DieselPunk_Impact_Cue
root=Delay_1
nodes [
 {id=Delay_1, children [Mixer_1], properties {delayMax=0.1, delayMin=0.1}}
 {id=Mixer_1, children [Random_2,Modulator_1], properties {inputVolume [0.6,1]}}
 {id=Modulator_1, children [Random_3], properties {pitchMax=1.1, pitchMin=0.98}}
 {id=Random_2, children [WavePlayer_3, WavePlayer_8, WavePlayer_9, WavePlayer_10], properties {weights [1,1,1,1]}}
 {id=Random_3, children [WavePlayer_11, WavePlayer_12, WavePlayer_13, WavePlayer_14], properties {weights [1,1,1,1]}}
 {id=WavePlayer_10, properties {soundWaveAssetPtr=/Game/Sounds/Fort_Weapons/SledgeHammers/Rocket_Hammer_Sledge_Impact_05}}
 {id=WavePlayer_11, properties {soundWaveAssetPtr=&/Diesel_Punk_Impact_04}}
 {id=WavePlayer_12, properties {soundWaveAssetPtr=&/Diesel_Punk_Impact_01}}
 {id=WavePlayer_13, properties {soundWaveAssetPtr=&/Diesel_Punk_Impact_02}}
 {id=WavePlayer_14, properties {soundWaveAssetPtr=&/Diesel_Punk_Impact_03}}
 {id=WavePlayer_3, properties {soundWaveAssetPtr=/Game/Sounds/Fort_Weapons/SledgeHammers/Rocket_Hammer_Sledge_Impact_01}}
 {id=WavePlayer_8, properties {soundWaveAssetPtr=/Game/Sounds/Fort_Weapons/SledgeHammers/Rocket_Hammer_Sledge_Impact_02}}
 {id=WavePlayer_9, properties {soundWaveAssetPtr=/Game/Sounds/Fort_Weapons/SledgeHammers/Rocket_Hammer_Sledge_Impact_04}}
]
properties {
 attenuationSettings=/Game/Sounds/Attenuation/Fort_Pickaxe_Impact_Att
 audioPropertiesSheet=/BRCosmetics/Sounds/Settings/PropertySheets/Pickaxe/APS_Cosmetic_Pickaxe_Impact
 bPrimeOnLoad=true
 maxDistance=4300
 pitchMultiplier=0.95
 soundClassObject=&Fort_Impacts
 attenuationOverrides {
  bAttenuateWithLPF=true
  distanceAlgorithm=Logarithmic
  falloffDistance=800
  lPFRadiusMax=800
  lPFRadiusMin=400
  attenuationShapeExtents {x=300}
 }
 concurrencyOverrides {bLimitToOwner=true, maxCount=2, resolutionRule=StopOldest}
 concurrencySet [&Player_Melee_Weapon_Conc]
}
